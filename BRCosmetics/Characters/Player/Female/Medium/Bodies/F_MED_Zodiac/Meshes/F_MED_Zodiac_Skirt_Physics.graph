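# Physics asset for the skirt of the female medium Zodiac player body, giving its frill bones capsule bodies for physics simulation.
# Capsules cover the pelvis, thighs, sub-thighs, two small front frill drivers and the back frill bones, with an extra pelvis frill on the left only.
# Back frills have angular limits against the pelvis or sub-thighs.
# Thigh-chain, front-driver and cross-links between back frills leave swing and twist free.
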PhysicsAsset /BRCosmetics/Characters/Player/Female/Medium/Bodies/F_MED_Zodiac/Meshes/F_MED_Zodiac_Skirt_Physics
bodies {
 driver_frill_frt_l [{shapeType=Capsule, length=0.505, radius=0.505, rotation=<rot 0 -90 0>}]
 driver_frill_frt_r [{shapeType=Capsule, length=0.505, radius=0.505, rotation=<rot 0 -90 0>}]
 dyn_frill_bck [{shapeType=Capsule, length=3.9913, radius=4.1291, center=<0 0 -3.5396>}]
 dyn_frill_bck_l [{shapeType=Capsule, length=6.8175, radius=4.25, center=<-3.1348 -0.0001 0>, rotation=<rot 90 0 0>}]
 dyn_frill_bck_r [{shapeType=Capsule, length=5.0675, radius=2, center=<3.1348 -0.0001 0>, rotation=<rot 90 0 0>}]
 dyn_pelvis_frill_back_l [{shapeType=Capsule, length=6.8175, radius=4.25, center=<-3.1348 -0.0001 0>, rotation=<rot 90 0 0>}]
 pelvis [{shapeType=Capsule, length=4.3155, radius=3.6737, rotation=<rot 0 90 0>}]
 sub_thigh_l [{shapeType=Capsule, length=1, radius=1}]
 sub_thigh_r [{shapeType=Capsule, length=1, radius=1}]
 thigh_l [{shapeType=Capsule, length=6.5, radius=6, center=<-10 0 0>, rotation=<rot 90 0 0>}]
 thigh_r [{shapeType=Capsule, length=5.7789, radius=6.7211, center=<10 0 0>, rotation=<rot 90 0 0>}]
}
constraints [
 {
  bone1Name=sub_thigh_l
  bone2Name=thigh_l
  swing1LimitDegrees=45
  swing1Motion=Free
  swing2LimitDegrees=45
  swing2Motion=Free
  twistLimitDegrees=45
  twistMotion=Free
 }
 {
  bone1Name=sub_thigh_r
  bone2Name=thigh_r
  swing1LimitDegrees=45
  swing1Motion=Free
  swing2LimitDegrees=45
  swing2Motion=Free
  twistLimitDegrees=45
  twistMotion=Free
 }
 {
  bone1Name=thigh_l
  bone2Name=pelvis
  swing1LimitDegrees=45
  swing1Motion=Free
  swing2LimitDegrees=45
  swing2Motion=Free
  twistLimitDegrees=45
  twistMotion=Free
 }
 {
  bone1Name=thigh_r
  bone2Name=pelvis
  swing1LimitDegrees=45
  swing1Motion=Free
  swing2LimitDegrees=45
  swing2Motion=Free
  twistLimitDegrees=45
  twistMotion=Free
 }
 {bone1Name=dyn_frill_bck, bone2Name=pelvis, swing1LimitDegrees=60.4286, swing2LimitDegrees=65.5714, twistLimitDegrees=45}
 {bone1Name=dyn_frill_bck_l, bone2Name=sub_thigh_l, swing1LimitDegrees=45, swing2LimitDegrees=45, twistLimitDegrees=45}
 {
  bone1Name=driver_frill_frt_l
  bone2Name=sub_thigh_l
  swing1LimitDegrees=45
  swing1Motion=Free
  swing2LimitDegrees=45
  swing2Motion=Free
  twistLimitDegrees=45
  twistMotion=Free
 }
 {bone1Name=dyn_frill_bck_r, bone2Name=sub_thigh_r, swing1LimitDegrees=45, swing2LimitDegrees=45, twistLimitDegrees=45}
 {
  bone1Name=driver_frill_frt_r
  bone2Name=sub_thigh_r
  swing1LimitDegrees=45
  swing1Motion=Free
  swing2LimitDegrees=45
  swing2Motion=Free
  twistLimitDegrees=45
  twistMotion=Free
 }
 {
  bone1Name=dyn_frill_bck_r
  bone2Name=dyn_frill_bck
  swing1LimitDegrees=30
  swing1Motion=Free
  swing2LimitDegrees=30
  swing2Motion=Free
  twistLimitDegrees=30
  twistMotion=Free
 }
 {
  bone1Name=dyn_frill_bck_l
  bone2Name=dyn_frill_bck
  swing1LimitDegrees=30
  swing1Motion=Free
  swing2LimitDegrees=30
  swing2Motion=Free
  twistLimitDegrees=30
  twistMotion=Free
 }
 {bone1Name=dyn_pelvis_frill_back_l, bone2Name=pelvis, swing1LimitDegrees=69, swing2LimitDegrees=60.4286, twistLimitDegrees=45}
 {
  bone1Name=dyn_pelvis_frill_back_l
  bone2Name=dyn_frill_bck
  swing1LimitDegrees=30
  swing1Motion=Free
  swing2LimitDegrees=30
  swing2Motion=Free
  twistLimitDegrees=30
  twistMotion=Free
 }
]
