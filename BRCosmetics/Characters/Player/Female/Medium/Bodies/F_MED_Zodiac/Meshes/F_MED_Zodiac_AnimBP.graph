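# Animation Blueprint for the F_MED_Zodiac character body that adds secondary motion to the pose copied from its attached parent mesh.
# A chain of AnimDynamics, constraints and bone offsets moves holster, shoes, grenades, jacket, frills and collar, with RigidBody skirt physics.
# The event graph drives jacket dynamics from jumping, backbling and DBNO, and correctives from head and thigh rotation.
# Both thigh readings write thigh_rot_l, leaving thigh_rot_r unset; several nodes are unconnected.
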AnimBlueprint /BRCosmetics/Characters/Player/Female/Medium/Bodies/F_MED_Zodiac/Meshes/F_MED_Zodiac_AnimBP
target_skeleton=&/F_MED_Zodiac_Skeleton
graphs [
 AnimGraph {nodes [
   Anim Dynamics - Bone: dyn_grenadeA_02(Alpha=1, ExternalForce="0,0,0") {
    angularDampingOverride=1
    angularSpringConstant=1500
    bAngularSpring=true
    bGravityOverrideInSimSpace=true
    bOverrideAngularDamping=true
    gravityScale=0.35
    simulationSpace=RootRelative
    alphaBoolBlend {blendOption=Linear}
    boundBone {boneName=dyn_grenadeA_02}
    physicsBodyDefinitions [{
      sphereCollisionRadius=0
      boundBone {boneName=dyn_grenadeA_02}
      boxExtents=<10 20 20>
      constraintSetup {angularLimitsMax=<0 0 0>, angularLimitsMin=<0 0 -40>, angularTarget=<0 0 0>}
      localJointOffset=<10 0 0>
     }]
   }
   Anim Dynamics - Bone: dyn_grenadeB_02(Alpha=1, ExternalForce="0,0,0") {
    angularDampingOverride=1
    angularSpringConstant=1500
    bAngularSpring=true
    bGravityOverrideInSimSpace=true
    bOverrideAngularDamping=true
    gravityScale=0.35
    simulationSpace=RootRelative
    alphaBoolBlend {blendOption=Linear}
    boundBone {boneName=dyn_grenadeB_02}
    physicsBodyDefinitions [{
      sphereCollisionRadius=0
      boundBone {boneName=dyn_grenadeB_02}
      boxExtents=<10 20 20>
      constraintSetup {angularLimitsMax=<0 0 0>, angularLimitsMin=<0 0 -40>, angularTarget=<0 0 0>}
      localJointOffset=<-10 0 0>
     }]
   }
   Anim Dynamics - Bone: dyn_grenadeC_02(Alpha=1, ExternalForce="0,0,0") {
    angularDampingOverride=1
    angularSpringConstant=1500
    bAngularSpring=true
    bGravityOverrideInSimSpace=true
    bOverrideAngularDamping=true
    gravityScale=0.35
    simulationSpace=RootRelative
    alphaBoolBlend {blendOption=Linear}
    boundBone {boneName=dyn_grenadeC_02}
    physicsBodyDefinitions [{
      sphereCollisionRadius=0
      boundBone {boneName=dyn_grenadeC_02}
      boxExtents=<10 20 20>
      constraintSetup {angularLimitsMax=<0 0 0>, angularLimitsMin=<0 0 -40>, angularTarget=<0 0 0>}
      localJointOffset=<-10 0 0>
     }]
   }
   n1 = "To Float (Boolean)"(InBool=Get IsDBNO() -> (IsDBNO:bool))
   "Constraint - Flap_Front_Driven_Bone (PP)"!(ComponentPose, Alpha=@n1, ConstraintWeights_0=0.5, ConstraintWeights_1:float=0.5) -> (Pose) {
    alphaBoolBlend {blendOption=Linear}
    boneToModify {boneName=Flap_Front_Driven_Bone}
    constraintSetup [{targetBone {boneName=dyn_frill_frt_l}}, {targetBone {boneName=dyn_frill_frt_l}}]
    constraintWeights [1,1]
   }
   n2 = "Constraint - dyn_frill_bck (PP)"!(ComponentPose, Alpha=@n1, ConstraintWeights_0=1, ConstraintWeights_1:float=1) -> (Pose) {
    alphaBoolBlend {blendOption=Linear}
    boneToModify {boneName=dyn_frill_bck}
    constraintSetup [{targetBone {boneName=dyn_frill_bck_l}}, {targetBone {boneName=dyn_frill_bck_r}}]
    constraintWeights [1,1]
   }
   "Transform (Modify) Bone - Bone: Flap_Front_Parent_Space"(ComponentPose=@n2, Alpha=@n1, Translation="0,0,0", Rotation="0,0,45", Scale="1,1,1") {
    rotationMode=BMM_Additive
    rotationSpace=BCS_BoneSpace
    alphaBoolBlend {blendOption=Linear}
    boneToModify {boneName=Flap_Front_Parent_Space}
    rotation {roll=45}
   }
   n3 = Get jacket_dynamic_alpha() -> (jacket_dynamic_alpha:double)
   n4 = Get jacket_ext_force() -> (jacket_ext_force:Vector)
   n5 = Anim Dynamics - Bone: dyn_holster(ComponentPose=Local To Component(LocalPose=Copy Pose From Mesh() {bUseAttachedParent=true}), Alpha=0.3, ExternalForce="0,0,0") {
    angularDampingOverride=1
    angularSpringConstant=5000
    bAngularSpring=true
    bGravityOverrideInSimSpace=true
    bOverrideAngularDamping=true
    gravityScale=0.35
    simulationSpace=RootRelative
    alphaBoolBlend {blendOption=Linear}
    boundBone {boneName=dyn_holster}
    physicsBodyDefinitions [{
      sphereCollisionRadius=0
      boundBone {boneName=dyn_holster}
      boxExtents=<20 20 20>
      constraintSetup {angularTargetAxis=AxisZ, angularLimitsMax=<0 10 0>, angularLimitsMin=<0 -10 -40>, angularTarget=<0 0 1>}
      localJointOffset=<10 0 0>
     }]
   }
   n6 = Anim Dynamics - Bone: dyn_holster_tass_01(ComponentPose=@n5, Alpha=1, ExternalForce="0,0,0") {
    angularDampingOverride=1
    angularSpringConstant=1500
    bAngularSpring=true
    bGravityOverrideInSimSpace=true
    bOverrideAngularDamping=true
    gravityScale=0.35
    simulationSpace=RootRelative
    alphaBoolBlend {blendOption=Linear}
    boundBone {boneName=dyn_holster_tass_01}
    physicsBodyDefinitions [{
      sphereCollisionRadius=0
      boundBone {boneName=dyn_holster_tass_01}
      boxExtents=<10 20 20>
      constraintSetup {angularLimitsMax=<8 15 0>, angularLimitsMin=<-8 -15 -40>, angularTarget=<0 0 0>}
      localJointOffset=<10 0 0>
     }]
   }
   n7 = Anim Dynamics - Bone: dyn_holster_tass_02(ComponentPose=@n6, Alpha=1, ExternalForce="0,0,0") {
    angularDampingOverride=1
    angularSpringConstant=1500
    bAngularSpring=true
    bGravityOverrideInSimSpace=true
    bOverrideAngularDamping=true
    gravityScale=0.35
    simulationSpace=RootRelative
    alphaBoolBlend {blendOption=Linear}
    boundBone {boneName=dyn_holster_tass_02}
    physicsBodyDefinitions [{
      sphereCollisionRadius=0
      boundBone {boneName=dyn_holster_tass_02}
      boxExtents=<10 20 20>
      constraintSetup {angularLimitsMax=<10 15 0>, angularLimitsMin=<-10 -15 -40>, angularTarget=<0 0 0>}
      localJointOffset=<10 0 0>
     }]
   }
   n8 = Anim Dynamics - Bone: dyn_holster_tass_03(ComponentPose=@n7, Alpha=1, ExternalForce="0,0,0") {
    angularDampingOverride=1
    angularSpringConstant=1500
    bAngularSpring=true
    bGravityOverrideInSimSpace=true
    bOverrideAngularDamping=true
    gravityScale=0.35
    simulationSpace=RootRelative
    alphaBoolBlend {blendOption=Linear}
    boundBone {boneName=dyn_holster_tass_03}
    physicsBodyDefinitions [{
      sphereCollisionRadius=0
      boundBone {boneName=dyn_holster_tass_03}
      boxExtents=<10 20 20>
      constraintSetup {angularLimitsMax=<5 20 0>, angularLimitsMin=<-5 -20 -40>, angularTarget=<0 0 0>}
      localJointOffset=<10 0 0>
     }]
   }
   n9 = Anim Dynamics - Bone: dyn_shoe_01_l(ComponentPose=@n8, Alpha=0.3, ExternalForce="0,0,0") {
    angularDampingOverride=1
    angularSpringConstant=1500
    bAngularSpring=true
    bGravityOverrideInSimSpace=true
    bOverrideAngularDamping=true
    gravityScale=0.35
    simulationSpace=RootRelative
    alphaBoolBlend {blendOption=Linear}
    boundBone {boneName=dyn_shoe_01_l}
    physicsBodyDefinitions [{
      sphereCollisionRadius=0
      boundBone {boneName=dyn_shoe_01_l}
      boxExtents=<10 19 22>
      constraintSetup {angularLimitsMax=<0 4 20>, angularLimitsMin=<0 -4 -15>, angularTarget=<0 0 0>}
      localJointOffset=<-10 0 0>
     }]
   }
   n10 = Anim Dynamics - Bone: dyn_shoe_02_l(ComponentPose=@n9, Alpha=0.7, ExternalForce="0,0,0") {
    angularDampingOverride=1
    angularSpringConstant=1500
    bAngularSpring=true
    bGravityOverrideInSimSpace=true
    bOverrideAngularDamping=true
    gravityScale=0.32
    simulationSpace=RootRelative
    alphaBoolBlend {blendOption=Linear}
    boundBone {boneName=dyn_shoe_02_l}
    physicsBodyDefinitions [{
      sphereCollisionRadius=0
      boundBone {boneName=dyn_shoe_02_l}
      boxExtents=<15 20 20>
      constraintSetup {angularLimitsMax=<0 8 20>, angularLimitsMin=<0 -8 -15>, angularTarget=<0 0 0>}
      localJointOffset=<-13 0 0>
     }]
   }
   n11 = Anim Dynamics - Bone: dyn_shoe_01_r(ComponentPose=@n10, Alpha=0.25, ExternalForce="0,0,0") {
    angularDampingOverride=1
    angularSpringConstant=1500
    bAngularSpring=true
    bGravityOverrideInSimSpace=true
    bOverrideAngularDamping=true
    gravityScale=0.35
    simulationSpace=RootRelative
    alphaBoolBlend {blendOption=Linear}
    boundBone {boneName=dyn_shoe_01_r}
    physicsBodyDefinitions [{
      sphereCollisionRadius=0
      boundBone {boneName=dyn_shoe_01_r}
      boxExtents=<13 15 18>
      constraintSetup {angularLimitsMax=<0 4 20>, angularLimitsMin=<0 -4 -15>, angularTarget=<0 0 0>}
      localJointOffset=<-10 0 0>
     }]
   }
   n12 = Anim Dynamics - Bone: dyn_shoe_02_r(ComponentPose=@n11, Alpha=0.65, ExternalForce="0,0,0") {
    angularDampingOverride=1
    angularSpringConstant=1500
    bAngularSpring=true
    bGravityOverrideInSimSpace=true
    bOverrideAngularDamping=true
    gravityScale=0.21
    simulationSpace=RootRelative
    alphaBoolBlend {blendOption=Linear}
    boundBone {boneName=dyn_shoe_02_r}
    physicsBodyDefinitions [{
      sphereCollisionRadius=0
      boundBone {boneName=dyn_shoe_02_r}
      boxExtents=<15 20 20>
      constraintSetup {angularLimitsMax=<0 8 20>, angularLimitsMin=<0 -8 -15>, angularTarget=<0 0 0>}
      localJointOffset=<-10 0 0>
     }]
   }
   n13 = "Constraint - parent_dyn_grenadeA_01 (P)"!(ComponentPose=@n12, Alpha=1, ConstraintWeights_0=1, ConstraintWeights_1:float=0) -> (Pose) {
    alphaBoolBlend {blendOption=Linear}
    boneToModify {boneName=parent_dyn_grenadeA_01}
    constraintSetup [{targetBone {boneName=upperarm_twist_01_l}}]
    constraintWeights [1,0]
   }
   n14 = "Constraint - parent_dyn_grenadeB_01 (PPP)"!(ComponentPose=@n13, Alpha=1, ConstraintWeights_0=0.1943, ConstraintWeights_1:float=0.1829, ConstraintWeights_2:float=0.6228) -> (Pose) {
    alphaBoolBlend {blendOption=Linear}
    boneToModify {boneName=parent_dyn_grenadeB_01}
    constraintSetup [{targetBone {boneName=upperarm_twist_02_l}}, {targetBone {boneName=upperarm_l}}, {targetBone {boneName=upperarm_twist_01_l}}]
    constraintWeights [1,1,1]
   }
   n15 = "Constraint - parent_dyn_grenadeC_01 (PPP)"!(ComponentPose=@n14, Alpha=1, ConstraintWeights_0=0.3171, ConstraintWeights_1:float=0.4223, ConstraintWeights_2:float=0.2606) -> (Pose) {
    alphaBoolBlend {blendOption=Linear}
    boneToModify {boneName=parent_dyn_grenadeC_01}
    constraintSetup [{targetBone {boneName=upperarm_twist_02_l}}, {targetBone {boneName=upperarm_l}}, {targetBone {boneName=upperarm_twist_01_l}}]
    constraintWeights [1,1,1]
   }
   n16 = Anim Dynamics - Bone: parent_dyn_grenadeA_01(ComponentPose=@n15, Alpha=0.3, ExternalForce="0,0,0") {
    angularDampingOverride=1
    angularSpringConstant=1500
    bAngularSpring=true
    bGravityOverrideInSimSpace=true
    bOverrideAngularDamping=true
    gravityScale=0.32
    simulationSpace=RootRelative
    alphaBoolBlend {blendOption=Linear}
    boundBone {boneName=parent_dyn_grenadeA_01}
    physicsBodyDefinitions [{
      sphereCollisionRadius=0
      boundBone {boneName=parent_dyn_grenadeA_01}
      boxExtents=<22 15 20>
      constraintSetup {twistAxis=AxisZ, angularLimitsMax=<10 10 5>, angularLimitsMin=<-10 -10 -5>, angularTarget=<0 0 0>}
      localJointOffset=<0 15 0>
     }]
   }
   n17 = Anim Dynamics - Bone: parent_dyn_grenadeB_01(ComponentPose=@n16, Alpha=0.35, ExternalForce="0,0,0") {
    angularDampingOverride=1
    angularSpringConstant=1500
    bAngularSpring=true
    bGravityOverrideInSimSpace=true
    bOverrideAngularDamping=true
    gravityScale=0.33
    simulationSpace=RootRelative
    alphaBoolBlend {blendOption=Linear}
    boundBone {boneName=parent_dyn_grenadeB_01}
    physicsBodyDefinitions [{
      sphereCollisionRadius=0
      boundBone {boneName=parent_dyn_grenadeB_01}
      boxExtents=<25 32 33>
      constraintSetup {twistAxis=AxisZ, angularLimitsMax=<10 10 5>, angularLimitsMin=<-10 -10 -5>, angularTarget=<0 0 0>}
      localJointOffset=<0 15 -0.01>
     }]
   }
   n18 = Anim Dynamics - Bone: parent_dyn_grenadeC_01(ComponentPose=@n17, Alpha=0.25, ExternalForce="0,0,0") {
    angularDampingOverride=1
    angularSpringConstant=1500
    bAngularSpring=true
    bGravityOverrideInSimSpace=true
    bOverrideAngularDamping=true
    gravityScale=0.35
    simulationSpace=RootRelative
    alphaBoolBlend {blendOption=Linear}
    boundBone {boneName=parent_dyn_grenadeC_01}
    physicsBodyDefinitions [{
      sphereCollisionRadius=0
      boundBone {boneName=parent_dyn_grenadeC_01}
      boxExtents=<15 20 20>
      constraintSetup {twistAxis=AxisZ, angularLimitsMax=<10 10 5>, angularLimitsMin=<-10 -10 -5>, angularTarget=<0 0 0>}
      localJointOffset=<0 10 0.01>
     }]
    relativeSpaceBone {boneName=upperarm_l}
   }
   n19 = "Transform (Modify) Bone - Bone: dyn_jacketBack"(ComponentPose=@n18, Alpha=1, Translation="0,5,0", Rotation="0,0,0", Scale="1,1,1") {
    translationMode=BMM_Additive
    translationSpace=BCS_BoneSpace
    alphaBoolBlend {blendOption=Linear}
    boneToModify {boneName=dyn_jacketBack}
    translation {y=5}
   }
   n20 = "Transform (Modify) Bone - Bone: dyn_jacketSideBack_l"(ComponentPose=@n19, Alpha=1, Translation="0,-4,0", Rotation="0,0,0", Scale="1,1,1") {
    translationMode=BMM_Additive
    translationSpace=BCS_BoneSpace
    alphaBoolBlend {blendOption=Linear}
    boneToModify {boneName=dyn_jacketSideBack_l}
    translation {y=-4}
   }
   n21 = "Transform (Modify) Bone - Bone: dyn_jacketSideBack_r"(ComponentPose=@n20, Alpha=1, Translation="0,4,0", Rotation="0,0,0", Scale="1,1,1") {
    translationMode=BMM_Additive
    translationSpace=BCS_BoneSpace
    alphaBoolBlend {blendOption=Linear}
    boneToModify {boneName=dyn_jacketSideBack_r}
    translation {y=4}
   }
   n22 = Anim Dynamics - Bone: dyn_jacketFront_r(ComponentPose=@n21, Alpha=1, ExternalForce=@n4) {
    angularDampingOverride=1
    angularSpringConstant=200
    bAngularSpring=true
    bGravityOverrideInSimSpace=true
    bOverrideAngularDamping=true
    simulationSpace=RootRelative
    alphaBoolBlend {blendOption=Linear}
    boundBone {boneName=dyn_jacketFront_r}
    componentAppliedLinearAccClamp=<2000 2000 3000>
    componentLinearAccScale=<0.8 0.8 0.8>
    componentLinearVelScale=<0.5 0.5 0.5>
    physicsBodyDefinitions [{
      sphereCollisionRadius=0
      boundBone {boneName=dyn_jacketFront_r}
      boxExtents=<15 3 15>
      constraintSetup {angularLimitsMax=<2 5 2>, angularLimitsMin=<-2 -5 -60>, angularTarget=<1 0 0>}
      localJointOffset=<10 0 0>
     }]
   }
   n23 = Anim Dynamics - Bone: dyn_jacketFront_l(ComponentPose=@n22, Alpha=1, ExternalForce=@n4) {
    angularDampingOverride=1
    angularSpringConstant=200
    bAngularSpring=true
    bGravityOverrideInSimSpace=true
    bOverrideAngularDamping=true
    simulationSpace=RootRelative
    alphaBoolBlend {blendOption=Linear}
    boundBone {boneName=dyn_jacketFront_l}
    componentAppliedLinearAccClamp=<2000 2000 3000>
    componentLinearAccScale=<0.8 0.8 0.8>
    componentLinearVelScale=<0.5 0.5 0.5>
    physicsBodyDefinitions [{
      sphereCollisionRadius=0
      boundBone {boneName=dyn_jacketFront_l}
      boxExtents=<15 3 15>
      constraintSetup {angularLimitsMax=<2 5 2>, angularLimitsMin=<-2 -5 -60>, angularTarget=<1 0 0>}
      localJointOffset=<-10 0 0>
     }]
   }
   n24 = Anim Dynamics - Bone: dyn_jacketSide_r(ComponentPose=@n23, Alpha=@n3, ExternalForce=@n4) {
    angularDampingOverride=1
    angularSpringConstant=200
    bAngularSpring=true
    bGravityOverrideInSimSpace=true
    bOverrideAngularDamping=true
    simulationSpace=RootRelative
    alphaBoolBlend {blendOption=Linear}
    boundBone {boneName=dyn_jacketSide_r}
    componentAppliedLinearAccClamp=<2000 2000 1000>
    componentLinearAccScale=<0.8 0.8 0.8>
    componentLinearVelScale=<0.5 0.5 0.5>
    physicsBodyDefinitions [{
      sphereCollisionRadius=0
      boundBone {boneName=dyn_jacketSide_r}
      boxExtents=<15 3 15>
      constraintSetup {angularLimitsMax=<2 5 2>, angularLimitsMin=<-2 -5 -60>, angularTarget=<1 0 0>}
      localJointOffset=<10 0 0>
     }]
   }
   n25 = Anim Dynamics - Bone: dyn_jacketSide_l(ComponentPose=@n24, Alpha=@n3, ExternalForce=@n4) {
    angularDampingOverride=1
    angularSpringConstant=200
    bAngularSpring=true
    bGravityOverrideInSimSpace=true
    bOverrideAngularDamping=true
    simulationSpace=RootRelative
    alphaBoolBlend {blendOption=Linear}
    boundBone {boneName=dyn_jacketSide_l}
    componentAppliedLinearAccClamp=<2000 2000 3000>
    componentLinearAccScale=<0.8 0.8 0.8>
    componentLinearVelScale=<0.5 0.5 0.5>
    physicsBodyDefinitions [{
      sphereCollisionRadius=0
      boundBone {boneName=dyn_jacketSide_l}
      boxExtents=<15 3 15>
      constraintSetup {angularLimitsMax=<2 5 2>, angularLimitsMin=<-2 -5 -60>, angularTarget=<1 0 0>}
      localJointOffset=<-10 0 0>
     }]
   }
   n26 = Anim Dynamics - Bone: dyn_jacketSideBack_r(ComponentPose=@n25, Alpha=@n3, ExternalForce=@n4) {
    angularDampingOverride=1
    angularSpringConstant=4000
    bAngularSpring=true
    bGravityOverrideInSimSpace=true
    bOverrideAngularDamping=true
    simulationSpace=RootRelative
    alphaBoolBlend {blendOption=Linear}
    boundBone {boneName=dyn_jacketSideBack_r}
    componentAppliedLinearAccClamp=<2000 2000 3000>
    componentLinearAccScale=<0.75 0.75 0.75>
    componentLinearVelScale=<0.5 0.5 0.5>
    physicsBodyDefinitions [{
      sphereCollisionRadius=0
      boundBone {boneName=dyn_jacketSideBack_r}
      boxExtents=<15 3 15>
      constraintSetup {angularLimitsMax=<2 5 2>, angularLimitsMin=<-2 -5 -68>, angularTarget=<1 0 0>}
      localJointOffset=<10 0 0>
     }]
   }
   n27 = Anim Dynamics - Bone: dyn_jacketSideBack_l(ComponentPose=@n26, Alpha=@n3, ExternalForce=@n4) {
    angularDampingOverride=1
    angularSpringConstant=4000
    bAngularSpring=true
    bGravityOverrideInSimSpace=true
    bOverrideAngularDamping=true
    simulationSpace=RootRelative
    alphaBoolBlend {blendOption=Linear}
    boundBone {boneName=dyn_jacketSideBack_l}
    componentAppliedLinearAccClamp=<2000 2000 3000>
    componentLinearAccScale=<0.75 0.75 0.75>
    componentLinearVelScale=<0.5 0.5 0.5>
    physicsBodyDefinitions [{
      sphereCollisionRadius=0
      boundBone {boneName=dyn_jacketSideBack_l}
      boxExtents=<15 3 15>
      constraintSetup {angularLimitsMax=<2 5 2>, angularLimitsMin=<-2 -5 -68>, angularTarget=<1 0 0>}
      localJointOffset=<-10 0 0>
     }]
   }
   n28 = Anim Dynamics - Bone: dyn_jacketBack(ComponentPose=@n27, Alpha=@n3, ExternalForce=@n4) {
    angularDampingOverride=1
    angularSpringConstant=4000
    bAngularSpring=true
    bGravityOverrideInSimSpace=true
    bOverrideAngularDamping=true
    simulationSpace=RootRelative
    alphaBoolBlend {blendOption=Linear}
    boundBone {boneName=dyn_jacketBack}
    componentAppliedLinearAccClamp=<2000 2000 3000>
    componentLinearAccScale=<0.75 0.75 0.75>
    componentLinearVelScale=<0.5 0.5 0.5>
    physicsBodyDefinitions [{
      sphereCollisionRadius=0
      boundBone {boneName=dyn_jacketBack}
      boxExtents=<15 3 15>
      constraintSetup {angularLimitsMax=<2 5 2>, angularLimitsMin=<-2 -5 -68>, angularTarget=<1 0 0>}
      localJointOffset=<10 0 0>
     }]
   }
   n29 = RigidBody(ComponentPose=@n28, Alpha=0.7) {
    bClampLinearTranslationLimitToRefPose=true
    bOverrideWorldGravity=true
    overridePhysicsAsset=&/F_MED_Zodiac_Skirt_Physics
    simulationSpace=BaseBoneSpace
    alphaBoolBlend {blendOption=Linear}
    baseBoneRef {boneName=root}
    componentAppliedLinearAccClamp=<1000 1000 1500>
    componentLinearAccScale=<1 1 1>
    componentLinearVelScale=<1 1 1>
    overrideWorldGravity {z=-980}
   }
   n30 = "Transform (Modify) Bone - Bone: driver_frill_frt_l"(ComponentPose=@n29, Alpha=Get thigh_rot_l(), Translation="2,8,2", Rotation="-15,30,5", Scale="1,1,1") {
    rotationMode=BMM_Additive
    rotationSpace=BCS_BoneSpace
    translationMode=BMM_Additive
    translationSpace=BCS_BoneSpace
    alphaBoolBlend {blendOption=Linear}
    boneToModify {boneName=driver_frill_frt_l}
    rotation=<rot -15 30 5>
    translation=<2 8 2>
   }
   n31 = "Transform (Modify) Bone - Bone: driver_frill_frt_r"(ComponentPose=@n30, Alpha=Get thigh_rot_r(), Translation="-4,-5,-4", Rotation="0.000,-65,0.000", Scale="1,1,1") {
    rotationMode=BMM_Additive
    rotationSpace=BCS_BoneSpace
    translationMode=BMM_Additive
    translationSpace=BCS_BoneSpace
    alphaBoolBlend {blendOption=Linear}
    boneToModify {boneName=driver_frill_frt_r}
    rotation {yaw=-65}
    translation=<-4 -5 -4>
   }
   n32 = Anim Dynamics - Bone: driver_frill_frt_l(ComponentPose=@n31, Alpha=0.95, ExternalForce="0,0,0") {
    angularDampingOverride=0.9999
    angularSpringConstant=1500
    bAngularSpring=true
    bGravityOverrideInSimSpace=true
    bOverrideAngularDamping=true
    gravityScale=0.31
    numSolverIterationsPostUpdate=2
    numSolverIterationsPreUpdate=8
    simulationSpace=RootRelative
    alphaBoolBlend {blendOption=Linear}
    boundBone {boneName=driver_frill_frt_l}
    componentAppliedLinearAccClamp=<800 800 800>
    componentLinearAccScale=<0.5 0.5 0.5>
    componentLinearVelScale=<0.5 0.5 0.5>
    physicsBodyDefinitions [{
      sphereCollisionRadius=0
      boundBone {boneName=driver_frill_frt_l}
      boxExtents=<15 25 25>
      constraintSetup {angularLimitsMax=<5 15 5>, angularLimitsMin=<-5 0 -45>, angularTarget=<0 0 0>}
      localJointOffset=<-11 -1 -2>
     }]
    relativeSpaceBone {boneName=driver_frill_frt_l}
   }
   n33 = Anim Dynamics - Bone: dyn_frill_frt_r(ComponentPose=@n32, Alpha=0.825, ExternalForce="0,0,0") {
    angularDampingOverride=1
    angularSpringConstant=1500
    bAngularSpring=true
    bGravityOverrideInSimSpace=true
    bOverrideAngularDamping=true
    gravityScale=0.35
    numSolverIterationsPostUpdate=2
    numSolverIterationsPreUpdate=8
    simulationSpace=RootRelative
    alphaBoolBlend {blendOption=Linear}
    boundBone {boneName=dyn_frill_frt_r}
    componentAppliedLinearAccClamp=<100 100 100>
    physicsBodyDefinitions [{
      sphereCollisionRadius=0
      boundBone {boneName=dyn_frill_frt_r}
      boxExtents=<10 20 20>
      constraintSetup {angularLimitsMax=<0 15 0>, angularLimitsMin=<0 -15 -40>, angularTarget=<0 0 0>}
      localJointOffset=<10 0 0>
     }]
   }
   n34 = "Transform (Modify) Bone - Bone: dyn_collarSide_r"(
    ComponentPose=@n33
    Alpha=Get head_rot_back() -> (head_rot_back:double)
    Translation="-1.25,0.311,0.000"
    Rotation="-19.684,50,10"
    Scale="1,1,1"
   ) {
    rotationMode=BMM_Additive
    rotationSpace=BCS_BoneSpace
    translationMode=BMM_Additive
    translationSpace=BCS_BoneSpace
    alphaBoolBlend {blendOption=Linear}
    alphaScaleBiasClamp {bInterpResult=true, interpSpeedDecreasing=25, interpSpeedIncreasing=25}
    boneToModify {boneName=dyn_collarSide_r}
    rotation=<rot -19.684 50 10>
    translation=<-1.25 0.311>
   }
   Output Pose(Result=Component To Local(ComponentPose=@n34)) {layerGroup=None}
   Get crouch_rigid_alpha() -> (crouch_rigid_alpha:double)
  ]}
 EventGraph {nodes [
   Event BlueprintUpdateAnimation()
   > n1 = Sequence!(execute) -> (then_0, then_1, then_2:Exec, then_3:Exec, then_4:Exec, then_5:Exec, then_6:Exec, then_7:Exec, then_8:Exec)
   n2 = TryGetFortPawn()
   > Is Valid(InputObject=@n2)
   n3 = Get bIsDBNO!(self:Fort Pawn Object Reference=@n2) -> (bIsDBNO)
   > Set IsDBNO(IsDBNO:bool=@n3) -> (Output_Get:bool)
   > Set NotDBNO(NotDBNO:bool=NOT Boolean(A=@n3)) -> (Output_Get:bool)
   n4 = Branch(execute=@n1.then_1, Condition=Get bIsJumping())
   > Delay(Duration=0.25)
   > Set jacket_ext_force(jacket_ext_force:Vector="0,0,1250") -> (Output_Get:Vector)
   > n5 = Delay(Duration=0.375)
   Do Once(execute=@n4.else, Reset=@n5.then)
   > Set jacket_ext_force(jacket_ext_force="0, 0, 0")
   n6 = Branch(execute=@n1.then_6, Condition=Get isWearingBackbling())
   > Set jacket_backBling(jacket_backBling:double=1) -> (Output_Get:double)
   > n7 = Set jacket_dynamic_alpha(jacket_dynamic_alpha:double=0.05) -> (Output_Get:double)
   Branch(execute=[@n8.then, @n7.then], Condition=Get IsDBNO())
   > Set jacket_dynamic_alpha(jacket_dynamic_alpha=0.15)
   Set jacket_backBling(execute=@n6.else, jacket_backBling=0)
   > n8 = Set jacket_dynamic_alpha(jacket_dynamic_alpha=1)
   Do Once(execute=@n1.then_7)
   > n9 = Cast To FortPlayerPawn(Object=TryGetPawnOwner())
   > Is Valid(InputObject=GetSkeletalMeshForPartType(self=@n9, PartType=Backpack))
   > Set isWearingBackbling(isWearingBackbling=true)
   n10 = Branch(execute=@n1.then_2, Condition=Get bIsCrouching())
   n11 = Get PawnSpeed2D()
   > n12 = Branch(Condition="float <= float"!(A:double=MapRangeClamped(Value=@n11, InRangeA=0, InRangeB=150), B:double=0) -> (ReturnValue:bool) {is=title K2Node_CallFunction})
   > Set crouch_speed_alpha(crouch_speed_alpha:double=1) -> (Output_Get:double)
   > Set crouch_moving(crouch_moving:bool=false) -> (Output_Get:bool)
   > Set crouch_rigid_alpha(crouch_rigid_alpha:double=0.2) -> (Output_Get:double)
   > Set move_speed_alpha(move_speed_alpha:double=0) -> (Output_Get:double)
   Set crouch_speed_alpha(execute=@n12.else, crouch_speed_alpha=0)
   > Set crouch_moving(crouch_moving=true)
   > Set crouch_rigid_alpha(crouch_rigid_alpha=0.55)
   > Set move_speed_alpha(move_speed_alpha=1)
   Set crouch_rigid_alpha(execute=@n10.else, crouch_rigid_alpha=0.8)
   n13 = BreakTransform(InTransform=GetDeltaTransformFromRefPose(self=GetOwnerSkeletalMeshComponent(), BoneName=thigh_l, BaseName=pelvis))
   Set thigh_rot_l(execute=@n1.then_4, thigh_rot_l=MapRangeClamped(Value=BreakRotator(InRot=@n13.Rotation).Yaw, InRangeA=0, InRangeB=-100))
   n14 = BreakTransform(InTransform=GetDeltaTransformFromRefPose(self=GetOwnerSkeletalMeshComponent(), BoneName=head, BaseName=spine_05))
   Set head_rot_back(execute=@n1.then_8, head_rot_back:double=MapRangeClamped(Value=BreakRotator(InRot=@n14.Rotation).Yaw, InRangeA=0, InRangeB=-55)) -> (Output_Get:double)
   n15 = BreakTransform(InTransform=GetDeltaTransformFromRefPose(self=GetOwnerSkeletalMeshComponent(), BoneName=thigh_r, BaseName=pelvis))
   Set thigh_rot_l(execute=@n1.then_5, thigh_rot_l=MapRangeClamped(Value=BreakRotator(InRot=@n15.Rotation).Yaw, InRangeA=0, InRangeB=-100))
   n16 = BreakTransform(InTransform=GetDeltaTransformFromRefPose(self=GetOwnerSkeletalMeshComponent(), BoneName=upperarm_r, BaseName=spine_05))
   Set arm_up_r(execute=@n1.then_3, arm_up_r:double=MapRangeClamped(Value=BreakRotator(InRot=@n16.Rotation).Yaw, InRangeA=80, InRangeB=138)) -> (Output_Get:double)
   Get Dynamic Material Instance(SkeletalMeshComponent=GetPartSkeletalMeshComponent())
   SetScalarParameterValue(ParameterName=Hide Element 01, Value=1)
  ]}
]
variables [
 arm_up_r {type=number}
 crouch_moving {type=boolean}
 crouch_rigid_alpha {type=number}
 crouch_speed_alpha {type=number}
 head_rot_back {type=number}
 IsDBNO {type=boolean}
 isWearingBackbling {type=boolean}
 jacket_backBling {type=number}
 jacket_dynamic_alpha {type=number}
 jacket_ext_force {type=Vector}
 move_speed_alpha {type=number}
 NotDBNO {type=boolean}
 thigh_rot_l {type=number}
 thigh_rot_r {type=number}
]
