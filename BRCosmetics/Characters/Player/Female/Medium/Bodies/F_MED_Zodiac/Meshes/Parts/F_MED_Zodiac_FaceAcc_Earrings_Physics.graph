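# Physics asset for the earrings face accessory on the F_MED_Zodiac female player body, covering the left and right dynamic earring bones.
# Each earring is a box body constrained to the head capsule, with moderate swing limits and a much tighter twist limit.
# A second capsule on neck_02 joins the head through a constraint with free swing and twist, giving body shapes for the head and neck beside the earrings.
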PhysicsAsset /BRCosmetics/Characters/Player/Female/Medium/Bodies/F_MED_Zodiac/Meshes/Parts/F_MED_Zodiac_FaceAcc_Earrings_Physics
bodies {
 dyn_earring_l [{shapeType=Box, extentX=5.2526, extentY=1.1158, extentZ=7.5834, center=<-5.3869 0.4274 -0.1776>, rotation=<rot 0 85.6123 -88.1967>}]
 dyn_earring_r [{shapeType=Box, extentX=5.2526, extentY=1.1158, extentZ=7.5834, center=<-5.0153 0.3989 0.1659>, rotation=<rot 0 -94.3877 -88.1968>}]
 head [{shapeType=Capsule, length=5.5, radius=5.0834, rotation=<rot -90 0 -30.0023>}]
 neck_02 [{shapeType=Capsule, length=4.969, radius=4.1144, center=<-1.0713 -1.8829 -0.0135>, rotation=<rot -0.0403 -89.9995 -89.9973>}]
}
constraints [
 {bone1Name=dyn_earring_l, bone2Name=head, swing1LimitDegrees=60, swing2LimitDegrees=60, twistLimitDegrees=12}
 {bone1Name=dyn_earring_r, bone2Name=head, swing1LimitDegrees=60, swing2LimitDegrees=60, twistLimitDegrees=12}
 {
  bone1Name=head
  bone2Name=neck_02
  swing1LimitDegrees=45
  swing1Motion=Free
  swing2LimitDegrees=45
  swing2Motion=Free
  twistLimitDegrees=45
  twistMotion=Free
 }
]
