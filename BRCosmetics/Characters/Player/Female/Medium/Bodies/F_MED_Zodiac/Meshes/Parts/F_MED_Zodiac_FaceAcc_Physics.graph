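# Physics asset for the face accessory mesh of the medium female Zodiac character, covering six dynamic hair-tip bones and the head.
# Each hair tip has a single capsule body, and the head has a small capsule as the shared parent body.
# Six constraints tie each tip directly to the head rather than to each other, with individual swing cones and a tight twist limit so the tips move independently within bounds.
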PhysicsAsset /BRCosmetics/Characters/Player/Female/Medium/Bodies/F_MED_Zodiac/Meshes/Parts/F_MED_Zodiac_FaceAcc_Physics
bodies {
 dyn_hairTip_01 [{shapeType=Capsule, length=2.75, radius=1.6218, center=<0 -0.01 0>, rotation=<rot -59.9996 -49.9992 0>}]
 dyn_hairTip_02 [{shapeType=Capsule, length=4, radius=2.276, center=<0 -0.2742 0>, rotation=<rot -69.9992 -110.0015 0>}]
 dyn_hairTip_03 [{shapeType=Capsule, length=2.25, radius=1.7612, center=<0 -0.0349 0>, rotation=<rot -79.9999 -69.9992 0.0001>}]
 dyn_hairTip_04 [{shapeType=Capsule, length=2.5, radius=2.7022, center=<0 -0.21 0>, rotation=<rot 0 0 -70.0009>}]
 dyn_hairTip_05 [{shapeType=Capsule, length=3.25, radius=1.6318, center=<0 -0.1 0>, rotation=<rot -69.9993 -100.0033 -0.0001>}]
 dyn_hairTip_06 [{shapeType=Capsule, length=2.25, radius=1.3818, center=<0 -0.1 0>, rotation=<rot -69.9993 -100.0033 -0.0001>}]
 head [{shapeType=Capsule, length=0.7856, radius=1, rotation=<rot 1.9577 -132.0811 -90>}]
}
constraints [
 {bone1Name=dyn_hairTip_01, bone2Name=head, swing1LimitDegrees=40.5714, swing2LimitDegrees=32, twistLimitDegrees=10}
 {bone1Name=dyn_hairTip_02, bone2Name=head, swing1LimitDegrees=54.2857, swing2LimitDegrees=20, twistLimitDegrees=10}
 {bone1Name=dyn_hairTip_03, bone2Name=head, swing1LimitDegrees=35.4286, swing2LimitDegrees=40.5714, twistLimitDegrees=10}
 {bone1Name=dyn_hairTip_04, bone2Name=head, swing1LimitDegrees=44, swing2LimitDegrees=42.2857, twistLimitDegrees=10}
 {bone1Name=dyn_hairTip_05, bone2Name=head, swing1LimitDegrees=28.5714, swing2LimitDegrees=28.5714, twistLimitDegrees=10}
 {bone1Name=dyn_hairTip_06, bone2Name=head, swing1LimitDegrees=35.4286, swing2LimitDegrees=40.5714, twistLimitDegrees=10}
]
