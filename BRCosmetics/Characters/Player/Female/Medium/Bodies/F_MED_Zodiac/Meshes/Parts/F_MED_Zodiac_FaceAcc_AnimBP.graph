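# Animation Blueprint for the Zodiac female face accessory mesh, adding physics-driven secondary motion to it.
# It copies pose and curves from the attached parent, then chains two RigidBody solvers, an earrings physics asset followed by the face accessory one, into the output.
# Both simulate relative to the root bone with damped component motion and overridden gravity.
# Left and right earring AnimDynamics nodes are unconnected to the output, and the event graph has no logic.
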AnimBlueprint /BRCosmetics/Characters/Player/Female/Medium/Bodies/F_MED_Zodiac/Meshes/Parts/F_MED_Zodiac_FaceAcc_AnimBP
target_skeleton=&/F_MED_Zodiac_FaceAcc_Skeleton
graphs [
 AnimGraph {nodes [
   n1 = Anim Dynamics - Bone: dyn_earring_l(Alpha=1, ExternalForce="0,0,0") {
    angularDampingOverride=0.8
    angularSpringConstant=1500
    bGravityOverrideInSimSpace=true
    bOverrideAngularDamping=true
    gravityScale=0.35
    simulationSpace=RootRelative
    alphaBoolBlend {blendOption=Linear}
    boundBone {boneName=dyn_earring_l}
    componentAppliedLinearAccClamp=<5000 5000 5000>
    componentLinearAccScale=<1 1 0.5>
    componentLinearVelScale=<1 1 0.5>
    physicsBodyDefinitions [{
      sphereCollisionRadius=0
      boundBone {boneName=dyn_earring_l}
      boxExtents=<10 20 20>
      constraintSetup {angularLimitsMax=<0 25 15>, angularLimitsMin=<0 -10 -40>, angularTarget=<0 0 0>}
      localJointOffset=<-5 0 0>
     }]
   }
   Anim Dynamics - Bone: dyn_earring_r(ComponentPose=@n1, Alpha=1, ExternalForce="0,0,0") {
    angularDampingOverride=0.99
    angularSpringConstant=1500
    bGravityOverrideInSimSpace=true
    bOverrideAngularDamping=true
    gravityScale=0.32
    simulationSpace=RootRelative
    alphaBoolBlend {blendOption=Linear}
    boundBone {boneName=dyn_earring_r}
    componentAppliedLinearAccClamp=<5000 5000 5000>
    componentLinearAccScale=<1 1 0.5>
    componentLinearVelScale=<1 1 0.5>
    physicsBodyDefinitions [{
      sphereCollisionRadius=0
      boundBone {boneName=dyn_earring_r}
      boxExtents=<13 20 20>
      constraintSetup {angularLimitsMax=<0 10 15>, angularLimitsMin=<0 -25 -40>, angularTarget=<0 0 0>}
      localJointOffset=<-8 0 0>
     }]
   }
   n2 = RigidBody(ComponentPose=Local To Component(LocalPose=Copy Pose From Mesh() {bCopyCurves=true, bUseAttachedParent=true})) {
    bClampLinearTranslationLimitToRefPose=true
    bOverrideWorldGravity=true
    overridePhysicsAsset=&/F_MED_Zodiac_FaceAcc_Earrings_Physics
    simulationSpace=BaseBoneSpace
    alphaBoolBlend {blendOption=Linear}
    baseBoneRef {boneName=root}
    componentAppliedLinearAccClamp=<1100 1100 1100>
    componentLinearAccScale=<0.5 0.5 0.35>
    componentLinearVelScale=<0.5 0.5 0.35>
    overrideWorldGravity {z=-980}
   }
   n3 = RigidBody(ComponentPose=@n2) {
    bClampLinearTranslationLimitToRefPose=true
    bOverrideWorldGravity=true
    overridePhysicsAsset=&/F_MED_Zodiac_FaceAcc_Physics
    simulationSpace=BaseBoneSpace
    alphaBoolBlend {blendOption=Linear}
    baseBoneRef {boneName=root}
    componentAppliedLinearAccClamp=<3000 3000 2000>
    componentLinearAccScale=<0.7 0.7 0.5>
    componentLinearVelScale=<0.7 0.7 0.5>
    overrideWorldGravity {z=-980}
   }
   Output Pose(Result=Component To Local(ComponentPose=@n3)) {layerGroup=None}
  ]}
 EventGraph {nodes [
   TryGetPawnOwner()
   Event BlueprintUpdateAnimation()
  ]}
]
variables []
